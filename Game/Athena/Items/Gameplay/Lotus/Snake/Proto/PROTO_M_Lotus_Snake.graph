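# Prototype translucent material for the Lotus Snake gameplay item, usable on mesh particles.
# A red-default Color vector parameter feeds the Emissive input of the shared MF_FX_EmissiveAdjustments material function.
# Depth testing is disabled, so the material renders through occluding geometry.
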Material /Game/Athena/Items/Gameplay/Lotus/Snake/Proto/PROTO_M_Lotus_Snake
graph {
 properties {
  bDisableDepthTest=true
  bUsedWithMeshParticles=true
  blendMode=BLEND_Translucent
  pixelDepthOffsetMode=PDOM_Legacy
  usedShadingModels=Default Lit
 }
 nodes [MaterialFunctionCall(Emissive=VectorParameter() {parameterName=Color, defaultValue=<rgba 1 0.0085 0 0>}.RGB) {materialFunction=&MF_FX_EmissiveAdjustments}]
}
